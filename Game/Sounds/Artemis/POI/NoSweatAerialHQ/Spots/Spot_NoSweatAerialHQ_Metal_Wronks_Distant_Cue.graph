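# Distant metal wronk ambient one-shot cue for the NoSweat Aerial HQ POI, reusing UFO crash site and generic metal stress waves.
# A Quality Level node plays a two-layer mix in its first two slots and a single, smaller random pool in the other two.
# The mix pairs randomized crash site wronks with dark, reverberant stress sounds, each through a Modulator.
# Output uses wide one-shot ambient attenuation and the outdoor ambient sound class.
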SoundCue /Game/Sounds/Artemis/POI/NoSweatAerialHQ/Spots/Spot_NoSweatAerialHQ_Metal_Wronks_Distant_Cue
root=QualityLevel_0
nodes [
 {id=Mixer_11, children [Modulator_1,Modulator_3], properties {inputVolume [0.75,0.75]}}
 {id=Modulator_1, children [Random_1], properties {pitchMax=1.1, pitchMin=0.8, volumeMax=2.5, volumeMin=2}}
 {id=Modulator_2, children [Random_2], properties {pitchMax=1.1, pitchMin=0.8, volumeMax=2.5, volumeMin=2}}
 {id=Modulator_3, children [Random_3]}
 {id=QualityLevel_0, children [Mixer_11,Mixer_11,Modulator_2,Modulator_2]}
 {
  id=Random_1
  children [WavePlayer_5, WavePlayer_6, WavePlayer_12, WavePlayer_13, WavePlayer_14, WavePlayer_15, WavePlayer_16, WavePlayer_18, WavePlayer_19]
  properties {weights [1,1,1,1,1,1,1,1,1]}
 }
 {id=Random_2, children [WavePlayer_25, WavePlayer_26, WavePlayer_27, WavePlayer_28, WavePlayer_29], properties {weights [1,1,1,1,1]}}
 {
  id=Random_3
  children [WavePlayer_30, WavePlayer_31, WavePlayer_32, WavePlayer_33, WavePlayer_34, WavePlayer_35]
  properties {weights [1,1,1,1,1,1]}
 }
 {
  id=WavePlayer_12
  properties {soundWaveAssetPtr=/Game/Sounds/Apollo/Biomes/UFO_Crash_Sites/AmbientOneshots/AMB_UFO_Crashsite_Metal_Wronk_Spot_02}
 }
 {
  id=WavePlayer_13
  properties {soundWaveAssetPtr=/Game/Sounds/Apollo/Biomes/UFO_Crash_Sites/AmbientOneshots/AMB_UFO_Crashsite_Metal_Wronk_Spot_03}
 }
 {
  id=WavePlayer_14
  properties {soundWaveAssetPtr=/Game/Sounds/Apollo/Biomes/UFO_Crash_Sites/AmbientOneshots/AMB_UFO_Crashsite_Metal_Wronk_Spot_04}
 }
 {
  id=WavePlayer_15
  properties {soundWaveAssetPtr=/Game/Sounds/Apollo/Biomes/UFO_Crash_Sites/AmbientOneshots/AMB_UFO_Crashsite_Metal_Wronk_Spot_05}
 }
 {
  id=WavePlayer_16
  properties {soundWaveAssetPtr=/Game/Sounds/Apollo/Biomes/UFO_Crash_Sites/AmbientOneshots/AMB_UFO_Crashsite_Metal_Wronk_Spot_06}
 }
 {
  id=WavePlayer_18
  properties {soundWaveAssetPtr=/Game/Sounds/Apollo/Biomes/UFO_Crash_Sites/AmbientOneshots/AMB_UFO_Crashsite_Metal_Wronk_Spot_07}
 }
 {
  id=WavePlayer_19
  properties {soundWaveAssetPtr=/Game/Sounds/Apollo/Biomes/UFO_Crash_Sites/AmbientOneshots/AMB_UFO_Crashsite_Metal_Wronk_Spot_08}
 }
 {
  id=WavePlayer_25
  properties {soundWaveAssetPtr=&AMB_UFO_Crashsite_Metal_Wronk_Spot_05}
 }
 {
  id=WavePlayer_26
  properties {soundWaveAssetPtr=&AMB_UFO_Crashsite_Metal_Wronk_Spot_08}
 }
 {
  id=WavePlayer_27
  properties {soundWaveAssetPtr=/Game/Sounds/Apollo/Biomes/UFO_Crash_Sites/AmbientOneshots/AMB_UFO_Crashsite_Metal_Wronk_Spot_09}
 }
 {
  id=WavePlayer_28
  properties {soundWaveAssetPtr=&AMB_UFO_Crashsite_Metal_Wronk_Spot_02}
 }
 {
  id=WavePlayer_29
  properties {soundWaveAssetPtr=/Game/Sounds/Apollo/Biomes/UFO_Crash_Sites/AmbientOneshots/AMB_UFO_Crashsite_Metal_Wronk_Spot_10}
 }
 {id=WavePlayer_30, properties {soundWaveAssetPtr=/Game/Sounds/Apollo/Biomes/Generic/AMB_Metal_Stress_Wronk_Dark_Dist_Verb_06}}
 {id=WavePlayer_31, properties {soundWaveAssetPtr=/Game/Sounds/Apollo/Biomes/Generic/AMB_Metal_Stress_Wronk_Dark_Dist_Verb_01}}
 {id=WavePlayer_32, properties {soundWaveAssetPtr=/Game/Sounds/Apollo/Biomes/Generic/AMB_Metal_Stress_Wronk_Dark_Dist_Verb_02}}
 {id=WavePlayer_33, properties {soundWaveAssetPtr=/Game/Sounds/Apollo/Biomes/Generic/AMB_Metal_Stress_Wronk_Dark_Dist_Verb_04}}
 {id=WavePlayer_34, properties {soundWaveAssetPtr=/Game/Sounds/Apollo/Biomes/Generic/AMB_Metal_Stress_Wronk_Dark_Dist_Verb_03}}
 {id=WavePlayer_35, properties {soundWaveAssetPtr=/Game/Sounds/Apollo/Biomes/Generic/AMB_Metal_Stress_Wronk_Dark_Dist_Verb_05}}
 {
  id=WavePlayer_5
  properties {soundWaveAssetPtr=&AMB_UFO_Crashsite_Metal_Wronk_Spot_09}
 }
 {
  id=WavePlayer_6
  properties {soundWaveAssetPtr=&AMB_UFO_Crashsite_Metal_Wronk_Spot_10}
 }
]
properties {
 attenuationSettings=/Game/Sounds/Attenuation/Fort_TileAmbient_OneShot_Wider_Att
 maxDistance=8000
 soundClassObject=&Fort_Ambient_Outdoors_Athena
 volumeMultiplier=0.576
 preEffectBusSends [{}]
}
